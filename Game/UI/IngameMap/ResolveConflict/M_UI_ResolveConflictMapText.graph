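# M_UI_ResolveConflictMapText is a translucent User Interface domain material, named for text on the in-game map's resolve-conflict UI.
# It holds only a Color1 vector parameter in the Color group, defaulting to opaque red, and a scalar constant of one.
# No links from these nodes to the material outputs are shown, and there are no textures or math nodes.
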Material /Game/UI/IngameMap/ResolveConflict/M_UI_ResolveConflictMapText
graph {
 properties {
  bUsedWithStaticMesh=false
  blendMode=BLEND_Translucent
  materialDomain=MD_UI
  pixelDepthOffsetMode=PDOM_Legacy
  usedShadingModels=Default Lit
 }
 nodes [
  Constant() {r=1}
  VectorParameter() {group=Color, parameterName=Color1, defaultValue=<rgba 1 0 0 1>}
 ]
}
